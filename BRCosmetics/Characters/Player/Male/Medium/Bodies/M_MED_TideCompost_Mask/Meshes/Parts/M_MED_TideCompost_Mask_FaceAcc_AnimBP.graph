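# Post-copy Anim Blueprint for the face accessory mesh of the TideCompost Mask male character body.
# It copies the pose from the attached parent mesh, then adjusts head_crown with a Transform (Modify) Bone node.
# Each update sets a bool when the current LOD is 1 or higher, enabling that node's additive translation offset.
# The node also holds a scale value, but its scale mode stays at the default Ignore.
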AnimBlueprint /BRCosmetics/Characters/Player/Male/Medium/Bodies/M_MED_TideCompost_Mask/Meshes/Parts/M_MED_TideCompost_Mask_FaceAcc_AnimBP
target_skeleton=&/M_MED_TideCompost_Mask_FaceAcc_Skeleton
graphs [
 AnimGraph {nodes [
   n1 = "Transform (Modify) Bone - Bone: head_crown"!(
    ComponentPose=Local To Component(LocalPose=Copy Pose From Mesh() {bUseAttachedParent=true})
    bAlphaBoolEnabled:bool=Get LOD1_and_up() -> (LOD1_and_up:bool)
    Translation="0,1,1.1"
    Rotation="0,0,0"
    Scale="1.6,1.6,1.6"
   ) -> (Pose) {alphaInputType=Bool, translationMode=BMM_Additive, boneToModify {boneName=head_crown}, scale=<1.6 1.6 1.6>, translation {y=1, z=1.1}}
   Output Pose(Result=Component To Local(ComponentPose=@n1)) {layerGroup=None}
  ]}
 EventGraph {nodes [
   Event BlueprintUpdateAnimation()
   > Set LOD1_and_up(LOD1_and_up:bool="integer >= integer"(A=Get CurrentLOD(), B=1)) -> (Output_Get:bool)
  ]}
]
variables [LOD1_and_up {type=boolean}]
